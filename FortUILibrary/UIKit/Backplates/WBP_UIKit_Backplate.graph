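# WBP_UIKit_Backplate is a UI Kit backplate Widget Blueprint that subclasses WBP_UIKit_Backplate_Base from the Backplate system.
# It declares no variables or widgets of its own.
# Its only logic is a PreConstruct event that forwards IsDesignTime to the parent, so any visuals or behavior come from the base class.
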WidgetBlueprint /FortUILibrary/UIKit/Backplates/WBP_UIKit_Backplate
graphs [EventGraph {nodes [
   n1 = Event PreConstruct()
   > Parent: PreConstruct(IsDesignTime=@n1.IsDesignTime)
  ]}]
variables []
widgets {
 info {
  inheritedWidgetCount=0
  parentClass=/FortUILibrary/Systems/Backplate/WBP_UIKit_Backplate_Base.WBP_UIKit_Backplate_Base_C
  rootWidgetClass=null
  widgetCount=0
 }
 widgets []
}
